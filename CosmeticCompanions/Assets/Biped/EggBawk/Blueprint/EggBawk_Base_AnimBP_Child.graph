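# Animation Blueprint for the EggBawk biped cosmetic companion, targeting the EggBawk biped skeleton.
# AllowsCinematicOverride returns false, and GetSequencerAnimSlotNames returns a single slot name, SequencerDefault.
# The event graph is empty and the file shows no variables or animation graph of its own.
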AnimBlueprint /CosmeticCompanions/Assets/Biped/EggBawk/Blueprint/EggBawk_Base_AnimBP_Child
target_skeleton=/CosmeticCompanions/Assets/Biped/EggBawk/Meshes/EggBawk_Biped_Skeleton
graphs [
 AllowsCinematicOverride {nodes [
   AllowsCinematicOverride() {is=title K2Node_FunctionEntry}
   > Return Node!(execute, ReturnValue:bool=false)
  ]}
 EventGraph {links [], nodes []}
 GetSequencerAnimSlotNames {nodes [
   GetSequencerAnimSlotNames() {is=title K2Node_FunctionEntry}
   > Return Node(ReturnValue:Array of Names=Make Array("[0]":Name=SequencerDefault) -> (Array:Array of Names))
  ]}
]
variables []
