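# Niagara effect for the EggBawk companion's egg-spawn emote, Long1 variant, in CosmeticCompanions.
# Five emitters, named as lightweight (LWE), cover airglow, sparkles and three smoke layers; only SystemState is listed as a module.
# User parameters expose a running flag, color curve and value, normalized notify progress, skeletal mesh and a chrome feather-off toggle.
# The listing shows no renderers or links from these parameters to the emitters.
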NiagaraSystem /CosmeticCompanions/Assets/Biped/EggBawk/FX/NS_Emote_EggBawk_EggSpawn_Long1
dependencies {usedDataInterfaces [], usedDynamicInputs [], usedModules [&SystemState]}
emitters {
 AirglowLWE {rendererClasses [], renderers []}
 SmokeLWE {rendererClasses [], renderers []}
 SmokeLWE001 {rendererClasses [], renderers []}
 SmokeLWE002 {rendererClasses [], renderers []}
 SparklesLWE {rendererClasses [], renderers []}
}
summary {
 systemName=NS_Emote_EggBawk_EggSpawn_Long1
 emitters [
  {emitterName=SparklesLWE, rendererClasses []}
  {emitterName=AirglowLWE, rendererClasses []}
  {emitterName=SmokeLWE, rendererClasses []}
  {emitterName=SmokeLWE001, rendererClasses []}
  {emitterName=SmokeLWE002, rendererClasses []}
 ]
 userVariables [
  User.bIsRunning {type=NiagaraBool}
  User.ColorCurve {
   type=NiagaraDataInterfaceColorCurve
   description=""
   defaultValue {
    struct=NiagaraExt_VariableValue_DataInterface
    value {
     dataInterface=&self:NiagaraDataInterfaceColorCurve_0
     dataInterfaceClass=&NiagaraDataInterfaceColorCurve
    }
   }
  }
  User.ColorCurveDataValue {type=NiagaraFloat, defaultValue=0, description=""}
  User.NormalizedNotifyProgress {type=NiagaraFloat, defaultValue=0, description=""}
  User.SkeletalMesh {
   type=NiagaraDataInterfaceSkeletalMesh
   defaultValue {
    struct=NiagaraExt_VariableValue_DataInterface
    value {dataInterface=&self:NiagaraDataInterfaceSkeletalMesh_0}
   }
  }
  User.TurnOffFeathersForChrome {type=NiagaraBool}
 ]
}
