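# Post-process AnimBlueprint for the EggBawk biped cosmetic companion that adds secondary motion to its feathers and wattles.
# The input pose passes through a curve-driven PoseAsset node, then six chained root-relative Anim Dynamics nodes for the featherA, woddle and left/right feather bones.
# Each uses angular springs, angle limits and reduced gravity, with DynamicsAlpha inverted by scale/bias so higher values reduce the simulation.
# The EventGraph is empty.
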AnimBlueprint /CosmeticCompanions/Assets/Biped/EggBawk/Meshes/EggBawk_Biped_Post_AnimBP
target_skeleton=&/EggBawk_Biped_Skeleton
graphs [
 AnimGraph {nodes [
   n1 = Input Pose()
   n2 = Get PoseAsset() -> (PoseAsset:Pose Asset Object Reference)
   n3 = Get DynamicsAlpha() -> (DynamicsAlpha:float)
   n4 = Anim Dynamics - Bone: featherA_01(
    ComponentPose=Local To Component(LocalPose=Pose!(PoseAsset:Pose Asset Object Reference=@n2, SourcePose:Pose Link Structure=@n1) -> (Pose:Pose Link Structure) {is=title AnimGraphNode_PoseBlendNode, blendOption=Linear})
    Alpha=@n3
    ExternalForce="0,0,0"
   ) {
    angularDampingOverride=1
    angularSpringConstant=18000
    bAngularSpring=true
    bOverrideAngularDamping=true
    bUseGravityOverride=true
    gravityScale=0.35
    simulationSpace=RootRelative
    alphaBoolBlend {blendOption=Linear}
    alphaScaleBiasClamp {bias=1, scale=-1}
    boundBone {boneName=featherA_01}
    componentLinearAccScale=<1 1 1>
    componentLinearVelScale=<1 1 1>
    gravityOverride {z=-980}
    physicsBodyDefinitions [{
      boundBone {boneName=featherA_01}
      boxExtents=<20 30 30>
      constraintSetup {angularLimitsMax=<2 8 8>, angularLimitsMin=<-2 -8 -8>, angularTarget=<1 0 0>}
      localJointOffset=<4 0 0>
     }]
   }
   n5 = Anim Dynamics - Bone: featherA_02(ComponentPose=@n4, Alpha=@n3, ExternalForce="0,0,0") {
    angularDampingOverride=1
    angularSpringConstant=18000
    bAngularSpring=true
    bOverrideAngularDamping=true
    bUseGravityOverride=true
    gravityScale=0.35
    simulationSpace=RootRelative
    alphaBoolBlend {blendOption=Linear}
    alphaScaleBiasClamp {bias=1, scale=-1}
    boundBone {boneName=featherA_02}
    componentLinearAccScale=<1 1 1>
    componentLinearVelScale=<1 1 1>
    gravityOverride {z=-980}
    physicsBodyDefinitions [{
      boundBone {boneName=featherA_02}
      boxExtents=<20 30 30>
      constraintSetup {angularLimitsMax=<2 8 8>, angularLimitsMin=<-2 -8 -8>, angularTarget=<1 0 0>}
      localJointOffset=<4 0 0>
     }]
   }
   n6 = Anim Dynamics - Bone: woddle_01_r(ComponentPose=@n5, Alpha=@n3, ExternalForce="0,0,0") {
    angularDampingOverride=1
    angularSpringConstant=15000
    bAngularSpring=true
    bOverrideAngularDamping=true
    bUseGravityOverride=true
    bUsePlanarLimit=false
    gravityScale=0.35
    simulationSpace=RootRelative
    alphaScaleBiasClamp {bias=1, scale=-1}
    boundBone {boneName=woddle_01_r}
    componentAppliedLinearAccClamp=<3000 3000 3000>
    componentLinearAccScale=<1 1 1>
    componentLinearVelScale=<1 1 1>
    gravityOverride {z=-980}
    physicsBodyDefinitions [{
      boundBone {boneName=woddle_01_r}
      boxExtents=<25 25 25>
      constraintSetup {angularLimitsMax=<0 5 45>, angularLimitsMin=<0 -5 -45>, angularTarget=<1 0 0>}
      localJointOffset=<-4 0 0>
     }]
   }
   n7 = Anim Dynamics - Bone: woddle_01_l(ComponentPose=@n6, Alpha=@n3, ExternalForce="0,0,0") {
    angularDampingOverride=1
    angularSpringConstant=15000
    bAngularSpring=true
    bOverrideAngularDamping=true
    bUseGravityOverride=true
    bUsePlanarLimit=false
    gravityScale=0.35
    simulationSpace=RootRelative
    alphaScaleBiasClamp {bias=1, scale=-1}
    boundBone {boneName=woddle_01_l}
    componentAppliedLinearAccClamp=<3000 3000 3000>
    componentLinearAccScale=<1 1 1>
    componentLinearVelScale=<1 1 1>
    gravityOverride {z=-980}
    physicsBodyDefinitions [{
      boundBone {boneName=woddle_01_l}
      boxExtents=<25 25 25>
      constraintSetup {angularLimitsMax=<0 5 45>, angularLimitsMin=<0 -5 -45>, angularTarget=<1 0 0>}
      localJointOffset=<4 0 0>
     }]
   }
   n8 = Anim Dynamics - Bone: feather_01_l(ComponentPose=@n7, Alpha=@n3, ExternalForce="0,0,0") {
    angularDampingOverride=1
    angularSpringConstant=6000
    bAngularSpring=true
    bOverrideAngularDamping=true
    bUseGravityOverride=true
    bUsePlanarLimit=false
    gravityScale=0.35
    simulationSpace=RootRelative
    alphaBoolBlend {blendOption=Linear}
    alphaScaleBiasClamp {bias=1, scale=-1}
    boundBone {boneName=feather_01_l}
    componentAppliedLinearAccClamp=<2000 2000 2000>
    componentLinearAccScale=<1 1 1>
    componentLinearVelScale=<1 1 1>
    gravityOverride {z=-980}
    physicsBodyDefinitions [{
      boundBone {boneName=feather_01_l}
      boxExtents=<23 23 23>
      constraintSetup {angularLimitsMax=<5 2 0>, angularLimitsMin=<-5 -35 -25>, angularTarget=<1 0 0>}
      localJointOffset=<4 0 0>
     }]
    simSpaceSettings {maxAngularAcceleration=50, maxAngularVelocity=10, simSpaceAngularAlpha=1}
   }
   n9 = Anim Dynamics - Bone: feather_01_r(ComponentPose=@n8, Alpha=@n3, ExternalForce="0,0,0") {
    angularDampingOverride=1
    angularSpringConstant=6000
    bAngularSpring=true
    bOverrideAngularDamping=true
    bUseGravityOverride=true
    bUsePlanarLimit=false
    gravityScale=0.35
    simulationSpace=RootRelative
    alphaBoolBlend {blendOption=Linear}
    alphaScaleBiasClamp {bias=1, scale=-1}
    boundBone {boneName=feather_01_r}
    componentAppliedLinearAccClamp=<2000 2000 2000>
    componentLinearAccScale=<1 1 1>
    componentLinearVelScale=<1 1 1>
    gravityOverride {z=-980}
    physicsBodyDefinitions [{
      boundBone {boneName=feather_01_r}
      boxExtents=<23 23 23>
      constraintSetup {angularLimitsMax=<5 2 25>, angularLimitsMin=<-5 -35 0>, angularTarget=<1 0 0>}
      localJointOffset=<4 0 0>
     }]
    simSpaceSettings {maxAngularAcceleration=50, maxAngularVelocity=10, simSpaceAngularAlpha=1}
   }
   Output Pose(Result=Component To Local(ComponentPose=@n9))
  ]}
 EventGraph {links [], nodes []}
]
variables [PoseAsset {type=PoseAsset, default=&/EggBawk_Biped_PoseAsset}]
